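; Base editor object for NeverCook quest tools that adds slot-based SaveGame persistence, configured by a slot name and save class.
; Init calls Load, which validates both settings, loads the slot or creates a new object, then passes it to PostLoad.
; Save runs only after a load has been attempted, re-fetches the object through Load, calls PrepareToSave and writes the slot.
; PostLoad and PrepareToSave are empty hooks.
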(fn SaveGame_DoesSaveExist
  (bind _returnvalue (SaveGame|DoesSaveGameExist (Variables|SaveGame|GetsaveGame_SlotName)))
  (return _returnvalue))

(fn SaveGame_Init
  (bind (_didsavealreadyexist _savegameobject) (Internal|SaveGameLoad))
  (SaveGame|SaveGamePostLoad _didsavealreadyexist _savegameobject))

(fn SaveGame_IsSaveDataValid
  (return (and (Utilities|String|NotEqualExactly(String) (Variables|SaveGame|GetsaveGame_SlotName)) (Utilities|IsValidClass (Variables|SaveGame|GetsaveGame_Class)))))

(fn SaveGame_Load
  (Variables|Internal|SetsaveGame_hasLoaded true)
  (bind _isvalid (Internal|SaveGameIsSaveDataValid))
  (if _isvalid
    (bind _returnvalue (SaveGame|DoesSaveGameExist (Variables|SaveGame|GetsaveGame_SlotName)))
    (if _returnvalue
      (bind _returnvalue_2 (SaveGame|LoadGamefromSlot (Variables|SaveGame|GetsaveGame_SlotName)))
      (bind _output_get (Variables|SaveGame|SetsaveGame_Object _returnvalue_2))
      (return true _output_get)
      (else
        (bind _returnvalue_3 (SaveGame|CreateSaveGameObject (Variables|SaveGame|GetsaveGame_Class)))
        (bind _output_get_2 (Variables|SaveGame|SetsaveGame_Object _returnvalue_3))
        (return false _output_get_2)))))

(fn SaveGame_PostLoad
  (params (SaveGameExisted :type Boolean) (SaveGameObject :type Save%20Game%20Object%20Reference)))

(fn SaveGame_PrepareToSave
  (params (SaveGameObject :type Save%20Game%20Object%20Reference)))

(fn SaveGame_Save
  (if (Variables|Internal|GetsaveGame_hasLoaded)
    (bind (_didsavealreadyexist _savegameobject) (Internal|SaveGameLoad))
    (Utilities|IsValid _savegameobject
      (:Is%20Valid
        (SaveGame|SaveGamePreparetoSave _savegameobject)
        (SaveGame|SaveGametoSlot _savegameobject (Variables|SaveGame|GetsaveGame_SlotName))))))
